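# SC_Dopey_Amb_Cue is a looping ambience cue for the Dopey event, assigned to the Fort_Events_Mapwide sound class.
# A single looping wave player plays Dopey_Amb_01 with boosted volume.
# Overridden attenuation replaces the referenced pickaxe-impact asset with a non-spatialized box zone using logarithmic falloff and distance-based reverb settings.
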SoundCue /Figment_S08/Athena/Sounds/Events/Dopey/SC_Dopey_Amb_Cue
root=WavePlayer_0
nodes [{id=WavePlayer_0, properties {bLooping=true, soundWaveAssetPtr=&/Dopey_Amb_01}}]
properties {
 attenuationSettings=/Game/Sounds/Attenuation/Fort_Weapon_Pickaxe_Player_Impact_Att
 bOverrideAttenuation=true
 maxDistance=1792
 soundClassObject=/Figment_S08/Athena/Sounds/Events/SoundClasses/Fort_Events_Mapwide
 volumeMultiplier=1.5
 attenuationOverrides {
  attenuationShape=Box
  bSpatialize=false
  distanceAlgorithm=Logarithmic
  falloffDistance=768
  reverbDistanceMax=1024
  reverbDistanceMin=512
  reverbWetLevelMax=10
  attenuationShapeExtents=<1024 1024 256>
 }
 concurrencyOverrides {maxCount=2, resolutionRule=PreventNew, volumeScale=0.8}
}
